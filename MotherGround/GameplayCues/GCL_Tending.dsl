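; GCL_Tending is a looping Gameplay Cue notify, apparently for egg tending VFX.
; Its looping-start event receives the target, cue parameters and spawned particle and audio component arrays.
; The event only prints an orange debug message saying the egg tending VFX is playing; the components go unused and the construction script is empty.
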(event EventOnLoopingStart(EitherVFX-preferthisversion)
  (params (MyTarget :type Actor%20Object%20Reference) (Parameters :type Gameplay%20Cue%20Parameters%20Structure%20%28by%20ref%29) (ParticleComponents :type Array%20of%20FXSystem%20Component%20Object%20References) (AudioComponents :type Array%20of%20Audio%20Component%20Object%20References))
  (Development|PrintString 0 "EGG TENDING VFX PLAYING NOW" true true "(R=1.000000,G=0.250938,B=0.009595,A=1.000000)"))

(fn ConstructionScript)
