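; GCN_Tending is a gameplay cue that plays egg-tending effects, with debug prints still in place.
; OnExecute stores the target as Gardener and calls the parent only when EffectCauser is valid.
; OnBurst feeds EffectCauser's location into a spawned particle's EggLocation parameter.
; OnActive attaches a borrowed Zero Point crystal dash Niagara system at the pelvis and calls the parent, only when the target's instigator is locally controlled.
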(fn OnActive
  (params (MyTarget :type Actor%20Object%20Reference) (Parameters :type Gameplay%20Cue%20Parameters%20Structure%20%28by%20ref%29))
  (bind (_normalizedmagnitude _rawmagnitude _effectcontext _matchedtagname _originaltag _aggregatedsourcetags _aggregatedtargettags _location _normal _instigator _effectcauser _sourceobject _physicalmaterial _gameplayeffectlevel _abilitylevel _targetattachcomponent) (Ability|GameplayCue|BreakGameplayCueParameters Parameters))
  (if (Pawn|IsLocallyControlled (Variables|/Script/Engine.Actor|GetInstigator MyTarget))
    (Development|PrintString 0 "IS LOCAL" true true "(R=0.000000,G=1.000000,B=0.141573,A=1.000000)")
    (Niagara|SpawnSystemAttached "/FNGameplayCues/Athena/Items/ForagedItems/ZeroPoint/Assets/Zero_Point_Crystals/FX/NS_Athena_ZeroPointCrystals_Dash.NS_Athena_ZeroPointCrystals_Dash" _targetattachcomponent "pelvis" "0, 0, 0" "0, 0, 0" "SnapToTarget")
    (return (|CallParentFunction MyTarget Parameters))
    (else
      (Development|PrintString 0 "NOT LOCAL" true true "(R=1.000000,G=0.049180,B=0.000000,A=1.000000)"))))

(fn OnBurst(EitherVFX-preferthisversion)
  (params (MyTarget :type Actor%20Object%20Reference) (Parameters :type Gameplay%20Cue%20Parameters%20Structure%20%28by%20ref%29) (ParticleComponents :type Array%20of%20FXSystem%20Component%20Object%20References) (AudioComponents :type Array%20of%20Audio%20Component%20Object%20References) (BurstCameraShakeInstance :type Camera%20Shake%20Base%20Object%20Reference) (BurstDecalInstance :type Decal%20Actor%20Object%20Reference))
  (bind (_normalizedmagnitude _rawmagnitude _effectcontext _matchedtagname _originaltag _aggregatedsourcetags _aggregatedtargettags _location _normal _instigator _effectcauser) (Ability|GameplayCue|BreakGameplayCueParameters Parameters))
  (Utilities|IsValid _effectcauser
    (:Is%20Valid
      (Effects|Components|ParticleSystem|SetVectorParameter (Utilities|Array|Get(acopy) ParticleComponents) "EggLocation" (Transformation|GetActorLocation _effectcauser))
      (return))))

(fn OnExecute
  (params (MyTarget :type Actor%20Object%20Reference) (Parameters :type Gameplay%20Cue%20Parameters%20Structure%20%28by%20ref%29))
  (bind (_normalizedmagnitude _rawmagnitude _effectcontext _matchedtagname _originaltag _aggregatedsourcetags _aggregatedtargettags _location _normal _instigator _effectcauser) (Ability|GameplayCue|BreakGameplayCueParameters Parameters))
  (LocalVariables|SetGardener MyTarget)
  (Utilities|IsValid _effectcauser
    (:Is%20Valid
      (Development|PrintString 0 (Utilities|String|ToString(Vector) (Transformation|GetActorLocation _effectcauser)))
      (Development|PrintString 0 "EGG TENDING VFX PLAYING NOW" true true "(R=1.000000,G=0.250938,B=0.009595,A=1.000000)")
      (return (|CallParentFunction MyTarget Parameters)))))
